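# Sound cue for the effect heard when entering a big bush hideable.
# A random node picks one of three equally weighted Enter wave variants, excluded from branch culling, and a modulator applies a slight random volume variation.
# It plays on the gadgets and items sound class with the shared mono trap attenuation, and overrides concurrency to cap simultaneous plays per owner.
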SoundCue /Game/Sounds/Fort_Hideables/Hideable_BigBush_Enter_Cue
root=Modulator_0
nodes [
 {id=Modulator_0, children [Random_0], properties {volumeMax=1.7, volumeMin=1.6}}
 {id=Random_0, children [WavePlayer_0,WavePlayer_2,WavePlayer_3], properties {bShouldExcludeFromBranchCulling=true, weights [1,1,1]}}
 {id=WavePlayer_0, properties {soundWaveAssetPtr=&/Hideable_BigBush_Enter_03}}
 {id=WavePlayer_2, properties {soundWaveAssetPtr=&/Hideable_BigBush_Enter_01}}
 {id=WavePlayer_3, properties {soundWaveAssetPtr=&/Hideable_BigBush_Enter_02}}
]
properties {
 attenuationSettings=/Game/Sounds/Attenuation/Fort_Traps_Mono_Att
 bOverrideConcurrency=true
 maxDistance=6000
 soundClassObject=&Fort_Gadgets_Items
 volumeMultiplier=0.1607
 concurrencyOverrides {bLimitToOwner=true, maxCount=2}
}
